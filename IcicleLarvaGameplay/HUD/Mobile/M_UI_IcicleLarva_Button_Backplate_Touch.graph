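# Translucent UI material for the touch button backplate in the IcicleLarva mobile HUD.
# One branch smooth-steps the icons-backplate texture over normalized UI UVs, blending blue to white by green and shaping alpha from blue times an Opacity parameter.
# Another samples the button-backplate texture with U mirrored around the center via Round, deriving tint and mask values.
# The graph shows no links to material outputs.
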Material /IcicleLarvaGameplay/HUD/Mobile/M_UI_IcicleLarva_Button_Backplate_Touch
graph {
 properties {bAutomaticallySetUsageInEditor=false, blendMode=BLEND_Translucent, materialDomain=MD_UI, usedShadingModels=Default Lit}
 nodes [
  n1 = TextureCoordinate()
  n2 = Round(None=ComponentMask(None=@n1) {r=true})
  n3 = Multiply(A=ComponentMask(None=@n1) {r=true}) {constB=2}
  n4 = Multiply(A=@n2, B=Add(A=Multiply(A=@n3) {constB=-1}) {constB=2})
  n5 = Multiply(A=OneMinus(None=@n2), B=@n3)
  n6 = TextureSample(UVs=AppendVector(A=Add(A=@n4, B=@n5), B=ComponentMask(None=@n1) {g=true})) {texture=&/T_UI_IcicleLarva_Button_Backplate_Touch}
  Add(A=Constant3Vector() {constant=<rgba 0.0666 0.1812 0.4072 1>}, B=@n6.G)
  n7 = Constant() {r=0.3}
  n8 = Constant() {r=0.4}
  n9 = TextureSample(UVs=MaterialFunctionCall() {materialFunction=&GetUserInterfaceUV}.Normalized UV) {texture=&/T_UI_IcicleLarva_IconsBackplate_Touch}
  LinearInterpolate(
   A=Constant3Vector() {constant=<rgba 0.159 0.491 0.9823 1>}
   B=Constant3Vector() {constant=<rgba 1 1 1 0>}
   Alpha=SmoothStep(Min=@n7, Max=@n8, Value=@n9.G)
  )
  Multiply(A=SmoothStep(Min=@n7, Max=@n8, Value=@n9.B), B=ScalarParameter() {defaultValue=0.75, parameterName=Opacity})
  Multiply(A=Add(A=@n6.R, B=@n6.R)) {constB=0.25}
 ]
}
