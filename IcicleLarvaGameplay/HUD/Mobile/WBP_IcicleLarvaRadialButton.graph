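# Mobile HUD radial button for IcicleLarva gameplay that shows a radial timer and numeric countdown for an effect matching its listened gameplay tag.
# Gameplay cue events start and count effects, hiding the display at zero, with a fallback timer that resets lingering counts.
# Tag-added and tag-removed listeners start and clear a looping timer, while a helper reads active effect durations from the ability system.
# Effect completion hides the display and plays a sound.
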WidgetBlueprint /IcicleLarvaGameplay/HUD/Mobile/WBP_IcicleLarvaRadialButton
animations [TapAnim {
  binding_count=0
  display_label=""
  length_seconds=0.1
  object=&self:TapAnim
  sequence {bindings []}
 }]
graphs [
 EventGraph {nodes [
   n1 = VerifyCueCountIsClear() {is=title K2Node_CustomEvent}
   > Branch(Condition="Not Equal (Integer)"!(A:int=Get CueCount() -> (CueCount:int=0), B:int=0) -> (ReturnValue:bool) {is=title K2Node_CallFunction})
   > Set CueCount(CueCount:int=0) -> (Output_Get:int=0)
   > HideSelf()
   HideSelf() {is=title K2Node_CustomEvent}
   > SetVisibility(
    self=[
     Get Image_Timer() -> (Image_Timer:Image Object Reference)
     Get Image_Background() -> (Image_Background:Image Object Reference)
     Get TextTimeRemaining() -> (TextTimeRemaining:Common Numeric Text Block Object Reference)
    ]
    InVisibility=Collapsed
   )
   n2 = Event Construct()
   n3 = OnCursorModeChanged!() -> (OutputDelegate, then, bCursorModeEnabled:bool=~, ActionName:Name, CursorModeContentWidget:User Widget Object Reference) {is=title K2Node_CustomEvent}
   n4 = Get FortHUDContext!() -> (ReturnValue:Fort HUDContext Object Reference) {is=title K2Node_GetBlueprintContextEx}
   Bind Event to On Cursor Mode Changed(execute=@n2.then, self:Fort HUDContext Object Reference=@n4, Delegate=@n3) {is=title K2Node_AddDelegate}
   > SetVisibility(self=[Get TextTimeRemaining(), Get Image_Timer()], InVisibility=Collapsed)
   > n5 = GetListenedGameplayTag!(execute, self:Radial Bacchus Button Base Object Reference) -> (then, ReturnValue:Gameplay Tag Structure)
   n6 = GetOwningPlayer()
   > n7 = WaitGameplayTagAddToActor!(execute:Exec, TargetActor:Actor Object Reference=@n6, Tag:Gameplay Tag Structure=@n5, OnlyTriggerOnce:bool=false) -> (then:Exec, AsyncTaskProxy:Ability Async Wait Gameplay Tag Added Object Reference, Added:Exec) {is=title K2Node_AsyncAction}
   > n8 = WaitGameplayTagRemoveFromActor!(execute:Exec, TargetActor:Actor Object Reference=@n6, Tag:Gameplay Tag Structure=@n5, OnlyTriggerOnce:bool=false) -> (then:Exec, AsyncTaskProxy:Ability Async Wait Gameplay Tag Removed Object Reference, Removed:Exec) {is=title K2Node_AsyncAction}
   Clear and Invalidate Timer by Handle(execute=@n8.Removed, Handle=Get CooldownEvent() -> (CooldownEvent:Timer Handle Structure))
   > HideSelf() {is=title K2Node_CallFunction}
   n9 = Set Timer by Event(execute=@n7.Added, Delegate=Create Event(), Time=0.05, bLooping=true)
   > Set CooldownEvent(CooldownEvent:Timer Handle Structure=@n9) -> (Output_Get:Timer Handle Structure)
   n10 = Event UpdateRemainingTime!() -> (OutputDelegate, then, Remaining:float, Duration:float) {is=title K2Node_Event}
   > SetCurrentValue!(execute, self:Common Numeric Text Block Object Reference=Get TextTimeRemaining(), NewValue:float="Max (Float)"(A=@n10.Remaining, B=0)) -> (then)
   > n11 = GetDynamicMaterial(self=Get Image_Timer())
   > SetScalarParameterValue(
    self=@n11
    ParameterName=Percentage
    Value="Clamp (Float)"(Value=float / float(A=@n10.Remaining, B=@n10.Duration) {is=title K2Node_CallFunction}, Min=0, Max=1)
   )
   Event Destruct()
   > Unbind Event from On Cursor Mode Changed(self:Fort HUDContext Object Reference=@n4, Delegate=@n3) {is=title K2Node_RemoveDelegate}
   n12 = Event ReceiveRadialProgressEvent_Blueprint!() -> (
    OutputDelegate
    then
    EventName:Name=None
    EventType:EGameplayCueEvent Enum=OnActive
    Parameters:"Gameplay Cue Parameters Structure (by ref)"
   ) {is=title K2Node_Event}
   n13 = BreakGameplayCueParameters!(Parameters:"Gameplay Cue Parameters Structure (by ref)"=@n12.Parameters) -> (
    NormalizedMagnitude:float
    RawMagnitude:float
    EffectContext:Gameplay Effect Context Handle Structure
    MatchedTagName:Gameplay Tag Structure
    OriginalTag:Gameplay Tag Structure
    AggregatedSourceTags:Gameplay Tag Container Structure
    AggregatedTargetTags:Gameplay Tag Container Structure
    Location:Vector
    Normal:Vector
    Instigator:Actor Object Reference
    EffectCauser:Actor Object Reference
    SourceObject:Object Reference
    PhysicalMaterial:Physical Material Object Reference
    GameplayEffectLevel:int=0
    AbilityLevel:int=0
    TargetAttachComponent:Scene Component Object Reference
    bReplicateLocationWhenUsingMinimalRepProxy:bool
   )
   > n14 = Branch(Condition="float <= float"!(A:double=@n13.RawMagnitude, B:double=0) -> (ReturnValue:bool) {is=title K2Node_CallFunction})
   > HideSelf()
   n15 = GetListenedGameplayTag(execute=@n14.else)
   > Branch(Condition="Equal (Name)"(A=GetTagName!(GameplayTag:"Gameplay Tag Structure (by ref)"=@n15) -> (ReturnValue:Name=None), B=@n12.EventName))
   > n16 = Switch on EGameplayCueEvent!(execute, Selection:EGameplayCueEvent Enum=@n12.EventType) -> (OnActive:Exec, WhileActive:Exec, Executed:Exec, Removed:Exec) {is=title K2Node_SwitchEnum}
   Set EffectDuration(execute=[@n16.Executed, @n16.WhileActive], EffectDuration:float=@n13.RawMagnitude) -> (Output_Get:float)
   > Set EffectStartTime(EffectStartTime:float=GetGameTimeInSeconds()) -> (Output_Get:float)
   > UpdateEffect(self:Radial Bacchus Button Base Object Reference)
   > SetVisibility(self=[Get TextTimeRemaining(), Get Image_Timer()], InVisibility=HitTestInvisible)
   > ForceVolatile!(execute, self:Widget Object Reference, bForce:bool=true) -> (then)
   > Set CueCount(CueCount=int + int(A=Get CueCount() -> (CueCount=0), B=1) -> (ReturnValue=0) {is=title K2Node_CommutativeAssociativeBinaryOperator}) -> (Output_Get=0)
   > Clear and Invalidate Timer by Handle(Handle=Get VerifyCueCountTimerHandle() -> (VerifyCueCountTimerHandle:Timer Handle Structure))
   > n17 = Set Timer by Event(
    Delegate=@n1
    Time=float + float(A=Get EffectDuration() -> (EffectDuration:float), B=1) {is=title K2Node_CommutativeAssociativeBinaryOperator}
    bLooping=false
   )
   > Set VerifyCueCountTimerHandle(VerifyCueCountTimerHandle:Timer Handle Structure=@n17) -> (Output_Get:Timer Handle Structure)
   > StartUpdateEffect(self:Radial Bacchus Button Base Object Reference)
   n18 = Set CueCount(execute=@n16.Removed, CueCount=int + int(A=Get CueCount() -> (CueCount=0), B=-1) -> (ReturnValue=0)) -> (Output_Get=0)
   > Branch(Condition="integer <= integer"!(A:int=@n18, B:int=0) -> (ReturnValue:bool) {is=title K2Node_CallFunction})
   > HideSelf()
   Event OnEffectStarted() {is=title K2Node_Event}
   > SetVisibility(self=[Get Image_Timer(), Get TextTimeRemaining()], InVisibility=HitTestInvisible)
   > ForceVolatile(bForce=true)
   Event OnEffectCompleted() {is=title K2Node_Event}
   > HideSelf()
   > PlaySound2D(Sound=Get CooldownCompleteSound() -> (CooldownCompleteSound:Sound Base Object Reference))
  ]}
 IsListenedTag {nodes [
   n1 = IsListenedTag!() -> (then, MatchedTag:Gameplay Tag Structure) {is=title K2Node_FunctionEntry}
   > n2 = GetListenedGameplayTag()
   > Return Node!(
    execute
    Match:bool="Equal (GameplayTag)"!(A:Gameplay Tag Structure=@n2, B:Gameplay Tag Structure=@n1) -> (ReturnValue:bool) {is=title K2Node_CallFunction}
   )
  ]}
 UpdatedBindCooldown {nodes [
   UpdatedBindCooldown() {is=title K2Node_FunctionEntry}
   > SetVisibility(self=[Get TextTimeRemaining(), Get Image_Background(), Get Image_Timer()], InVisibility=HitTestInvisible)
   > n1 = GetListenedGameplayTag()
   > n2 = GetActiveEffectsWithAllTags!(
    execute
    self:Ability System Component Object Reference=GetFortAbilitySystemComponent!(Actor:Actor Object Reference=GetOwningPlayerPawn!(self:User Widget Object Reference) -> (ReturnValue:Pawn Object Reference)) -> (ReturnValue:Fort Ability System Component Object Reference)
    Tags_GameplayTags:Array of Gameplay Tag Structures=Make Array("[0]"=@n1)
   ) -> (then, ReturnValue:Array of Active Gameplay Effect Handle Structures)
   > Branch(Condition="integer > integer"(A=Length(TargetArray:Array of Active Gameplay Effect Handle Structures=@n2)))
   n3 = "Get (a ref)"(Array:Array of Active Gameplay Effect Handle Structures=@n2, Dimension 1=0) -> (Output:"Active Gameplay Effect Handle Structure (by ref)")
   > n4 = GetActiveGameplayEffectRemainingDuration!(execute, ActiveHandle:Active Gameplay Effect Handle Structure=@n3) -> (then, ReturnValue:float)
   > n5 = GetActiveGameplayEffectTotalDuration!(execute, ActiveHandle:Active Gameplay Effect Handle Structure=@n3) -> (then, ReturnValue:float)
   > n6 = GetDynamicMaterial(self=Get Image_Timer())
   > SetScalarParameterValue(self=@n6, ParameterName=Percentage, Value=float / float(A=@n4, B=@n5) {is=title K2Node_PromotableOperator})
   > SetCurrentValue(self=Get TextTimeRemaining(), NewValue="Max (Float)"(A=@n4, B=0))
  ]}
]
variables [
 CooldownCompleteSound {type=SoundBase, default=/IcicleLarvaGameplay/Assets/Audio/Waves/Laser_Eyes/IcicleLarva_Laser_Impact_Start_01}
 CooldownEvent {type=TimerHandle}
 CueCount {type=integer}
 DefaultIcon {type=PaperSprite, default=null}
 Pressed Color {type=SlateColor, default {colorUseRule=UseColor_Specified, specifiedColor=<rgba 0 0 0 1>}}
 Unpressed Color {type=SlateColor, default {colorUseRule=UseColor_Specified, specifiedColor=<rgba 1 1 1 1>}}
 VerifyCueCountTimerHandle {type=TimerHandle}
]
widgets {
 info {inheritedWidgetCount=1, parentClass=/Script/FortniteUI.RadialBacchusButtonBase, rootWidgetClass=&Overlay, widgetCount=10}
 widgets [
  Overlay_0 {widgetClassPath=&Overlay, widget=Overlay_0}
  ButtonSizeBox {
   widgetClassPath=&SizeBox
   bIsVariable=true
   parent=Overlay_0
   widget=ButtonSizeBox
   properties {bOverride_HeightOverride=true, bOverride_WidthOverride=true, heightOverride=200, widthOverride=200}
   slot {class=OverlaySlot, properties {horizontalAlignment=HAlign_Center, verticalAlignment=VAlign_Center, padding=<ltrb 4 2 4 2>}}
  }
  ImageSizeBox {
   widgetClassPath=&SizeBox
   bIsVariable=true
   parent=ButtonSizeBox
   widget=ImageSizeBox
   slot {class=SizeBoxSlot, properties {horizontalAlignment=HAlign_Center, verticalAlignment=VAlign_Center}}
  }
  ScaleBox_4 {widgetClassPath=&ScaleBox, parent=ImageSizeBox, widget=ScaleBox_4, slot {class=SizeBoxSlot}}
  SizeBox_1 {
   widgetClassPath=&SizeBox
   parent=ScaleBox_4
   widget=SizeBox_1
   properties {bOverride_HeightOverride=true, bOverride_WidthOverride=true, heightOverride=128, widthOverride=128}
   slot {class=ScaleBoxSlot}
  }
  Overlay_1 {widgetClassPath=&Overlay, parent=SizeBox_1, widget=Overlay_1, slot {class=SizeBoxSlot}}
  Image_Background {
   widgetClassPath=&Image
   bIsVariable=true
   parent=Overlay_1
   widget=Image_Background
   properties {
    visibility=Collapsed
    brush {
     bIsDynamicallyLoaded=false
     drawAs=Image
     imageType=FullColor
     mirroring=NoMirror
     resourceName=None
     resourceObject=/FortUILibrary/UIKit/Buttons/Materials/M_UIKit_RadialGlow
     tiling=NoTile
     imageSize=<128 128>
     margin=<ltrb 0 0 0 0>
     outlineSettings {
      bUseBrushTransparency=false
      roundingType=HalfHeightRadius
      width=0
      color {colorUseRule=UseColor_Specified, specifiedColor=<rgba 0 0 0 0>}
      cornerRadii=<0 0 0 0>
     }
     tintColor {colorUseRule=UseColor_Specified, specifiedColor=<rgba 1 1 1 1>}
     uVRegion {bIsValid=false, max=<0 0>, min=<0 0>}
    }
    colorAndOpacity=<rgba 0 0 0 0.5>
   }
   slot {class=OverlaySlot, properties {horizontalAlignment=HAlign_Fill, verticalAlignment=VAlign_Fill}}
  }
  Image_Timer {
   widgetClassPath=&Image
   bIsVariable=true
   parent=Overlay_1
   widget=Image_Timer
   properties {
    visibility=SelfHitTestInvisible
    brush {
     bIsDynamicallyLoaded=false
     drawAs=Image
     imageType=NoImage
     mirroring=NoMirror
     resourceName=None
     resourceObject=/Game/Athena/HUD/M_HealProgress
     tiling=NoTile
     imageSize=<105 105>
     margin=<ltrb 0 0 0 0>
     outlineSettings {
      bUseBrushTransparency=false
      roundingType=HalfHeightRadius
      width=0
      color {colorUseRule=UseColor_Specified, specifiedColor=<rgba 0 0 0 0>}
      cornerRadii=<0 0 0 0>
     }
     tintColor {colorUseRule=UseColor_Specified, specifiedColor=<rgba 1 1 1 1>}
     uVRegion {bIsValid=false, max=<0 0>, min=<0 0>}
    }
    renderTransform {angle=0, scale=<1.125 1.125>, shear=<0 0>, translation=<0 0>}
   }
   slot {class=OverlaySlot, properties {horizontalAlignment=HAlign_Center, verticalAlignment=VAlign_Center}}
  }
  TextTimeRemaining {
   widgetClassPath=/Script/CommonUI.CommonNumericTextBlock
   bIsVariable=true
   parent=Overlay_1
   widget=TextTimeRemaining
   properties {
    currentNumericValue=1.1
    justification=Center
    text=Text Block
    font {
     bForceMonospaced=false
     bMaterialIsStencil=false
     fontMaterial=null
     fontObject=&NotoSans
     letterSpacing=0
     monospacedWidth=1
     size=30
     skewAmount=0
     typefaceFontName=NotoSans-Regular
     outlineSettings {
      bApplyOutlineToDropShadows=false
      bMiteredCorners=false
      bSeparateFillAlpha=false
      outlineMaterial=null
      outlineSize=0
      outlineColor=<rgba 0 0 0 1>
     }
    }
   }
   slot {class=OverlaySlot, properties {horizontalAlignment=HAlign_Fill, verticalAlignment=VAlign_Center}}
  }
  TutorialHighlight {widgetClassPath=&TutorialHighlight}
 ]
}
